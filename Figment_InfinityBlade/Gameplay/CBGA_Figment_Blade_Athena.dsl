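; Building gameplay actor Blueprint for the Figment Infinity Blade, which by its name and sound path appears to be the pullable sword in Athena.
; Its only active logic plays the BRSword pull-success cue at the actor's location on EndPlay.
; That sound does not check the end-play reason, so it is requested on any removal, not just a pull.
; BeginPlay, overlap and Tick hold no logic, and the construction script only calls its parent.
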(event EventBeginPlay
  (|CallParentFunction) ; This node is disabled and will not be called.\nDrag off pins to build functionality.
)

(event Collision|EventActorBeginOverlap
  (params (OtherActor :type Actor%20Object%20Reference)))

(event EventTick
  (params (DeltaSeconds :type Float%20%28single-precision%29)))

(event EventEndPlay
  (params (EndPlayReason :type EEndPlayReason%20Enum :default "Destroyed"))
  (Audio|PlaySoundatLocation "/Game/Sounds/Fort_GamePlay_Sounds/BRSword/AB_Pull_End_Success_Cue.AB_Pull_End_Success_Cue" (Transformation|GetActorLocation)))

(fn ConstructionScript
  (|CallParentFunction))
