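; Passive setup ability for the Infinity Blade wielder that grants starting health, regenerates health or shield, and toggles HUD elements.
; On activation it equips the best weapon unless skipped, heals a first wielder, binds damage events and starts timers.
; A looping tick applies health, then shield regen effects, pausing while downed or damage-tagged.
; Delayed checks work around a latency pickaxe bug by force-equipping the sword or dropping all items.
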(event Ability|EventActivateAbility
  (bind _firstwieldertag (Variables|Default|GetFirstWielderTag))
  (|CallParentFunction)
  (bind _asplayer_pawn_athena (Utilities|Casting|CastToPlayerPawn_Athena (Ability|GetAvatarActorfromActorInfo))
    (:then
      (bind _output_get (Variables|Default|SetPlayerPawn _asplayer_pawn_athena))
      ;[1 1 1] Initialize.
      (if (not (ScalableFloat|GetValueasBool (Variables|Default|GetSkipForceEquipPickaxe)))
        (Class|FortPawn|EquipBestWeapon _output_get))
      (CallFunction|DisplayHUD false)
      (EventDispatchers|BindEventtoOnDamaged (Variables|Default|GetPlayerPawn) (delegate Custom|EventDamaged))
      ;[1 1 1] Fill up Health and Shields if this is initial equip.
      (if (GameplayTags|HasMatchingGameplayTag (Variables|Default|GetPlayerPawn) _firstwieldertag)
        (Tags|RemoveNonReplicatedTagtoPawn (Variables|Default|GetPlayerPawn) _firstwieldertag)
        (Ability|ApplyGameplayEffectToOwner "/Game/Athena/Items/Gameplay/Sword/GE_Sword_Starting_HealthPool.GE_Sword_Starting_HealthPool_C" 0) ; GE Starting Heals for Initial Wielder.
      )
      ;[1 1 1] Assign TickRate through rows in AthenaGameData
      (Variables|Default|SetTickRate (ScalableFloat|GetValueatLevel (Variables|Default|GetTickRateRow)))
      ;[1 1 1] Setup timer at the speed of tick rate
      (Utilities|Time|SetTimerbyEvent (EventDispatchers|CreateEvent :!event Custom|ApplyHealthOrShield) (Variables|Default|GetTickRate) true)
      (Utilities|Time|SetTimerbyEvent (delegate Custom|ForceEquipWeaponDefinition) (ScalableFloat|GetValueatLevel (Variables|Default|GetForceEquipWeaponDelay)))
      ; Workaround to fix a bug where the player can equip the pickaxe while picking up the Infinity Blade, especially when connection latency is too high [FORT-1000858].
      (Utilities|Time|SetTimerbyEvent (delegate Custom|CheckForTags) (ScalableFloat|GetValueatLevel (Variables|Default|GetCheckForPickaxeDelay))))
    (:CastFailed
      (Development|PrintString "Avatar Actor is NOT PlayerPawn_Athena"))))

(event Custom|CheckForTags
  ; Workaround to fix a bug where the player can equip the pickaxe while picking up the Infinity Blade, especially when connection latency is too high [FORT-1000858].
  (if (GameplayTags|HasMatchingGameplayTag (Variables|Default|GetPlayerPawn) "(TagName=\"Granted.ItemEquipped.Melee.Harvest\")")
    (Inventory|DropAllItems (Fortnite|GetFortPlayerControllerfromActor (Variables|Default|GetPlayerPawn)) 0 0 (ScalableFloat|GetValueasBool (Variables|Default|GetKeepBuildingMaterialOnDropAll)) true :!class /Script/FortniteGame.FortPlayerController)))

(event Custom|ForceEquipWeaponDefinition
  (bind _playerpawn (Variables|Default|GetPlayerPawn))
  (bind _asfort_player_controller (Utilities|Casting|CastToFortPlayerController (Pawn|GetController _playerpawn)))
  (if (GameplayTags|HasMatchingGameplayTag _playerpawn "(TagName=\"Granted.ItemEquipped.Melee.Harvest\")")
    (Inventory|ExecuteInventoryItemDefinition :self _asfort_player_controller :ItemDefinition "/Figment_InfinityBlade/Gameplay/AGID_Figment_Sword.AGID_Figment_Sword" :Delay 0.0 :bForceExecute true :bActivateSlotAfterSettingFocused true)))

(event Custom|EventDamaged
  (params (DamagedActor :type Actor%20Object%20Reference) (Damage :type Float%20%28single-precision%29) (InstigatedBy :type Controller%20Object%20Reference) (DamageCauser :type Actor%20Object%20Reference) (HitLocation :type Vector) (FHitComponent :type Primitive%20Component%20Object%20Reference) (BoneName :type Name) (Momentum :type Vector))
  ;[1 1 1] Apply Regen Cooldown when damaged.
  (Ability|ApplyGameplayEffectToOwner "/Game/Athena/Items/Gameplay/Sword/GE_DamagedTag.GE_DamagedTag_C" 0))

(event Ability|EventFailedToActivatePassiveAbility
  (params (ActorInfo :type Gameplay%20Ability%20Actor%20Info%20Structure))
  (|CallParentFunction))

(event Ability|EventOnEndAbility
  (params (bWasCancelled :type Boolean))
  (CallFunction|DisplayHUD true))

(event Custom|ApplyHealthOrShield
  (bind _playerpawn (Variables|Default|GetPlayerPawn))
  ;[1 1 1] Grant health or shield if not on cooldown.
  (if (GameplayTags|HasMatchingGameplayTag _playerpawn "(TagName=\"Athena.Blade.Damaged\")")
    _
    (elif (Health|IsDBNO _playerpawn)
      _
      (elif (< (FortPawn|GetHealth _playerpawn) (FortPawn|GetMaxHealth _playerpawn))
        ;;[0.2686 1 0.3384] Health
        (Ability|ApplyGameplayEffectToOwner "/Game/Athena/Items/Gameplay/Sword/GE_Athena_Blade_HealthRegen.GE_Athena_Blade_HealthRegen_C" 0)
        (else
          ;;[0.286 0.7239 1] Shield
          (Ability|ApplyGameplayEffectToOwner "/Game/Athena/Items/Gameplay/Sword/GE_Athena_Blade_Shield_Regen.GE_Athena_Blade_Shield_Regen_C" 0))))))

(fn DisplayHUD
  (params (Show :type Boolean))
  (bind _playerpawn (Variables|Default|GetPlayerPawn)
    (:then
      (Utilities|IsValid (Fortnite|GetFortPlayerControllerfromActor _playerpawn)
        (:Is%20Valid
          (HUD|SetHUDElementVisibilitywithContext (Fortnite|GetFortPlayerControllerfromActor _playerpawn) (GameplayTags|MakeGameplayTagContainerfromTag "(TagName=\"HUD.Athena.HidePrimaryForAbility\")") (not Show) "None" "")
          (return))))
    (:else)))
